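# Empty is a masked, Default Lit material among the blockout materials for vehicle cosmetics.
# Its graph holds a single Constant node at its default value, with no textures, parameters or other logic.
# The name and masked blend mode suggest a blank or hidden-slot placeholder, but the file does not show which input the constant feeds.
# Automatic usage-flag setting in the editor is turned off.
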Material /VehicleCosmetics/Materials/BlockOut_Materials/Vehicle_BO_Materials/Empty
graph {
 properties {
  bAutomaticallySetUsageInEditor=false
  blendMode=BLEND_Masked
  previewMesh=/Engine/Transient.MaterialEditorMeshComponent_5
  usedShadingModels=Default Lit
 }
 nodes [Constant()]
}
